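# MetaSound Source preset of MSS_EngineAudio_Off that provides the mono, one-shot engine-off sound for the Cyclone vehicle.
# The parent supplies the playback graph, while this preset fills its wave array with the Cyclone off recording and sets the output volume input.
# Vehicle movement attenuation, engine-off concurrency sets and the exterior vehicle sound class govern playback.
# Shared vehicle engine modulators drive volume, high-pass and low-pass.
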MetaSoundSource /VehicleAudio/EngineAudio/Cyclone/MSS_EngineAudio_Cyclone_Off_Preset
preset_parent=/VehicleAudio/Patches/MSS_EngineAudio_Off
inputs [
 In Array {type=WaveAsset:Array, access=Constructor}
 OutputVolume {type=Float, access=Constructor}
 UE.Source.OnPlay {type=Trigger}
]
interfaces [UE.OutputFormat.Mono:1.0, UE.Source.OneShot:1.0, UE.Source:1.1]
outputs [
 "UE.OutputFormat.Mono.Audio:0" {type=Audio}
 UE.Source.OneShot.OnFinished {type=Trigger}
]
pages [Default {input_defaults {OutputVolume=0, UE.Source.OnPlay=null, In Array [&/REV_Cyclone_Off_Mono]}}]
properties {
 assetImportData=&self:AssetImportData
 attenuationSettings=/VehicleAudio/Settings/Vehicle_Movement_Att
 audioPropertiesSheet=/VehicleAudio/PropertySheets/APS_Cosmetic_Vehicle_Engine_Off
 maxDistance=12512
 sampleRate=48000
 soundClassObject=&Fort_Vehicles_Exterior
 concurrencySet [/VehicleAudio/Settings/Vehicle_Cosmetic_Off_Conc, /VehicleAudio/Settings/Vehicle_Cosmetic_Ignition_Off_Conc]
 modulationSettings {
  highpassRouting=Union
  lowpassRouting=Union
  volumeRouting=Union
  highpassModulationDestination {bEnableModulation=true, modulators [&PP_FortVehicles_Engine_HPF]}
  lowpassModulationDestination {bEnableModulation=true, modulators [&Patch_FortVehicles_Engine_LPF]}
  volumeModulationDestination {
   bEnableModulation=true
   modulators [&Patch_FortVehicles_Engine_Vol, &DM_VehicleEnginePatch]
  }
 }
}
